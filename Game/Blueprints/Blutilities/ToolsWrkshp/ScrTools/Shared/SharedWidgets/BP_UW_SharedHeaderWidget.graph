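# Shared header UserWidget for screen tools under Blutilities, providing a bordered vertical container that is filled with expandable boxes at runtime.
# AddExpandableBox constructs a BP_UW_ExpandableBox, adds it to MainContainerVertBox and returns it to the caller.
# ConfigureExpandableBox and the event graph are empty stubs.
# Label variables for Actions, Tool Information and Shortcuts boxes are declared, but no graph shown reads them.
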WidgetBlueprint /Game/Blueprints/Blutilities/ToolsWrkshp/ScrTools/Shared/SharedWidgets/BP_UW_SharedHeaderWidget
graphs [
 AddExpandableBox {nodes [
   AddExpandableBox() {is=title K2Node_FunctionEntry}
   > n1 = Construct Object from Class(Class=/Game/Blueprints/Blutilities/ToolsWrkshp/ScrTools/Shared/SharedWidgets/ExpandableBox/BP_UW_ExpandableBox.BP_UW_ExpandableBox_C) -> (ReturnValue:BP UW Expandable Box Object Reference)
   > n2 = Set AddedExpandableBox(AddedExpandableBox:BP UW Expandable Box Object Reference=@n1) -> (Output_Get:BP UW Expandable Box Object Reference)
   > AddChildToVerticalBox(self=Get MainContainerVertBox() -> (MainContainerVertBox:Vertical Box Object Reference), Content=@n2)
   > Return Node!(
    execute
    New Expandable Box:BP UW Expandable Box Object Reference=Get AddedExpandableBox() -> (AddedExpandableBox:BP UW Expandable Box Object Reference)
   )
  ]}
 ConfigureExpandableBox {nodes [ConfigureExpandableBox() {is=title K2Node_FunctionEntry}]}
 EventGraph {nodes [
   Event PreConstruct()
   Event Construct()
   Event Tick()
  ]}
]
variables [
 Actions Box Label {type=string, default=Actions}
 Readout Box Label {type=string, default=Tool Information}
 Shortcut Box Label {type=string, default=Shortcuts}
]
widgets {
 info {inheritedWidgetCount=0, parentClass=&UserWidget, rootWidgetClass=&Border, widgetCount=2}
 widgets [
  Border_51 {
   widgetClassPath=&Border
   widget=Border_51
   properties {
    background {
     bIsDynamicallyLoaded=false
     drawAs=Image
     imageType=FullColor
     mirroring=NoMirror
     resourceName=None
     resourceObject=null
     tiling=NoTile
     imageSize=<256 256>
     margin=<ltrb 0 0 0 0>
     outlineSettings {
      bUseBrushTransparency=false
      roundingType=HalfHeightRadius
      width=0
      color {colorUseRule=UseColor_Specified, specifiedColor=<rgba 0 0 0 0>}
      cornerRadii=<0 0 0 0>
     }
     tintColor {colorUseRule=UseColor_Specified, specifiedColor=<rgba 0.0469 0.0469 0.0469 0>}
     uVRegion {bIsValid=false, max=<0 0>, min=<0 0>}
    }
    brushColor=<rgba 0 0 0 1>
    contentColorAndOpacity=<rgba 0.9473 0.9473 0.9473 1>
    padding=<ltrb 0 2 0 2>
   }
  }
  MainContainerVertBox {
   widgetClassPath=&VerticalBox
   bIsVariable=true
   parent=Border_51
   widget=MainContainerVertBox
   slot {class=BorderSlot, properties {padding=<ltrb 0 2 0 2>}}
  }
 ]
}
